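; Looping gameplay cue for LoudKiwi's "stacking applied" effect on structures, driving the density of its stacking particle effect.
; On looping start it calls the parent, caches the first spawned particle component as NS_Stacking and sets its SpawnDensity parameter to 1.
; Each burst sets SpawnDensity on that cached component to the cue's gameplay effect level, so the running effect is updated in place.
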(event EventOnLoopingStart(EitherVFX-preferthisversion)
  (params (MyTarget :type Actor%20Object%20Reference) (Parameters :type Gameplay%20Cue%20Parameters%20Structure%20%28by%20ref%29) (ParticleComponents :type Array%20of%20FXSystem%20Component%20Object%20References) (AudioComponents :type Array%20of%20Audio%20Component%20Object%20References))
  (bind _output (Utilities|Array|Get(acopy) ParticleComponents))
  (|CallParentFunction MyTarget Parameters ParticleComponents AudioComponents)
  (Variables|Default|SetNS_Stacking _output)
  (Effects|Components|ParticleSystem|SetIntParameter _output "SpawnDensity" 1))

(fn OnBurst
  (params (MyTarget :type Actor%20Object%20Reference) (Parameters_NormalizedMagnitude :type Float%20%28single-precision%29) (Parameters_RawMagnitude :type Float%20%28single-precision%29) (Parameters_EffectContext :type Gameplay%20Effect%20Context%20Handle%20Structure) (Parameters_MatchedTagName :type Gameplay%20Tag%20Structure) (Parameters_OriginalTag :type Gameplay%20Tag%20Structure) (Parameters_AggregatedSourceTags :type Gameplay%20Tag%20Container%20Structure) (Parameters_AggregatedTargetTags :type Gameplay%20Tag%20Container%20Structure) (Parameters_Location :type Vector :default "0, 0, 0") (Parameters_Normal :type Vector :default "0, 0, 0") (Parameters_Instigator :type Actor%20Object%20Reference) (Parameters_EffectCauser :type Actor%20Object%20Reference) (Parameters_SourceObject :type Object%20Reference) (Parameters_PhysicalMaterial :type Physical%20Material%20Object%20Reference) (Parameters_GameplayEffectLevel :type Integer) (Parameters_AbilityLevel :type Integer) (Parameters_TargetAttachComponent :type Scene%20Component%20Object%20Reference) (Parameters_bReplicateLocationWhenUsingMinimalRepProxy :type Boolean))
  (Effects|Components|ParticleSystem|SetIntParameter (Variables|Default|GetNS_Stacking) "SpawnDensity" Parameters_GameplayEffectLevel)
  (return false))

(fn ConstructionScript)
