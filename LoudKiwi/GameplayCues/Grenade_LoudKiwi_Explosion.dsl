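; Gameplay cue for the LoudKiwi sticky grenade explosion, playing effects based on whether the hit actor is a water body.
; Water hits spawn the CubeSplitter Niagara effect at world origin, fed water normal, depth and body index, plus debris sound at the projectile.
; Other hits spawn the same effect at the projectile with global scale and custom depth.
; Both show a "Gunshot" sound indicator from the first burst sound, then call the parent.
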(fn OnExecute
  (params (MyTarget :type Actor%20Object%20Reference) (Parameters :type Gameplay%20Cue%20Parameters%20Structure%20%28by%20ref%29))
  (bind (_normalizedmagnitude _rawmagnitude _effectcontext) (Ability|GameplayCue|BreakGameplayCueParameters Parameters))
  (bind (_bblockinghit _binitialoverlap _time _distance _location _impactpoint _normal _impactnormal _physmat _hitactor) (Collision|BreakHitResult (Ability|EffectContext|GetHitResult _effectcontext)))
  (bind _asb_prj_sticky_loud_kiwi_athena (Utilities|Casting|CastToB_Prj_Sticky_LoudKiwi_Athena MyTarget))
  (LocalVariables|SetExplosionLocation (Transformation|GetActorLocation _asb_prj_sticky_loud_kiwi_athena))
  (Utilities|IsA(soft) _hitactor "/Script/FortniteGame.FortWaterBodyActor"
    (:True
      ;[0.075 1 1 0.5] WATER EXPLOSION EFFECT
      (bind _returnvalue (Effects|Components|ParticleSystem|SpawnFXAtLocation 0 "/LoudKiwi/Assets/NS_LoudKiwi_CubeSplitter_GroundExplosion.NS_LoudKiwi_CubeSplitter_GroundExplosion" "0, 0, 0" (Math|Rotator|MakeRotfromX (.WaterPlaneNormal (FX|GetLastWaterSurfaceInfo (Variables|/Script/FortniteGame.FortProjectileBase|GetWaterInteractionComponent _asb_prj_sticky_loud_kiwi_athena) :!class /Script/FortniteGame.FortWaterInteractionComponent)))))
      (Effects|Components|ParticleSystem|SetFloatParameter _returnvalue "WaterDepth" (.WaterDepth (FX|GetLastWaterSurfaceInfo (Variables|/Script/FortniteGame.FortProjectileBase|GetWaterInteractionComponent _asb_prj_sticky_loud_kiwi_athena) :!class /Script/FortniteGame.FortWaterInteractionComponent)))
      (Effects|Components|ParticleSystem|SetFloatParameter _returnvalue "WaterBodyIndex" (Math|Conversions|ToFloat(Integer) (.WaterBodyIdx (FX|GetLastWaterSurfaceInfo (Variables|/Script/FortniteGame.FortProjectileBase|GetWaterInteractionComponent _asb_prj_sticky_loud_kiwi_athena) :!class /Script/FortniteGame.FortWaterInteractionComponent))))
      (Audio|SpawnSoundatLocation 0 "/Game/Athena/Sounds/Weapons/Explosions/Expl_Debris_Water_Cue.Expl_Debris_Water_Cue" (LocalVariables|GetExplosionLocation)))
    (:False
      ;[1 0.1 0.1 0.5] GROUND EXPLOSION EFFECT
      (bind _returnvalue_2 (Effects|Components|ParticleSystem|SpawnFXAtLocation _asb_prj_sticky_loud_kiwi_athena "/LoudKiwi/Assets/NS_LoudKiwi_CubeSplitter_GroundExplosion.NS_LoudKiwi_CubeSplitter_GroundExplosion" (LocalVariables|GetExplosionLocation) (Math|Rotator|MakeRotfromZ (LocalVariables|GetExplosionLocation))))
      (Effects|Components|ParticleSystem|SetFloatParameter _returnvalue_2 "Global Scale" (Variables|Default|GetExplosion%20Global%20Scale))
      (Rendering|SetRenderCustomDepth _returnvalue_2 true)))
  (CallFunction|ShowSoundIndicator (LocalVariables|GetExplosionLocation))
  (return (|CallParentFunction MyTarget Parameters)))

(fn ShowSoundIndicator
  (params (Location :type Vector))
  (bind _bursteffectdata (Variables|BurstEffects|GetBurstEffectData))
  (bind (_burstparticles _burstsounds) (Utilities|Struct|BreakFortBurstEffectData _bursteffectdata))
  (bind (_condition _attachment _soundcue) (Utilities|Struct|BreakFortGameplayCueAudioInfo (Utilities|Array|Get(aref) _burstsounds)))
  (if (> (Utilities|Array|Length _burstsounds) 0)
    (Fortnite|ShowSoundIndicatorLocal :Location Location :MaxAudibleDistance (Variables|/Script/Engine.SoundBase|GetMaxDistance _soundcue) :IndicatorType "Gunshot" :Tint "(R=0.000000,G=0.000000,B=0.000000,A=0.000000)" :SoundBase _soundcue :AttenuationMaxRangeMultiplier 1.0 :AttenuationIntensityMultiplier 1.0)))
